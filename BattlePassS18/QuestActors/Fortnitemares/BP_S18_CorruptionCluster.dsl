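; Season 18 Fortnitemares quest actor for a corruption cluster with tentacles, extending a parent that supplies the hide and unhide timelines.
; On BeginPlay it puts a dynamic tentacle material on the pre-interaction mesh and sets its wiggle displacement vector.
; Both timeline update events call the parent, then use DoOnce gates to play crystal success and spawn/despawn Niagara effects at the actor's location, with the spawn effect scaled up.
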(event EventBeginPlay
  (|CallParentFunction)
  ;[1 1 1] Update Tenticle material on static mesh
  (bind _returnvalue (Rendering|Material|CreateDynamicMaterialInstance (Variables|Default|GetREF_TenticleMat) "newInstance" "None" :!class /Script/Engine.KismetMaterialLibrary))
  (bind _output_get (Variables|Default|SetREF_TenticleWPO _returnvalue))
  (Rendering|Material|SetMaterial (Variables|Components|GetSM_Quest_PreInteraction) 0 _output_get)
  (Rendering|Material|SetVectorParameterValue (Variables|Default|GetREF_TenticleWPO) "WiggleDisplacementAmt" (Math|Color|MakeColor 6e+01 6e+01 46.0) :!class /Script/Engine.MaterialInstanceDynamic))

(event Collision|EventActorBeginOverlap
  (params (OtherActor :type Actor%20Object%20Reference))
  (|CallParentFunction OtherActor) ; This node is disabled and will not be called.\nDrag off pins to build functionality.
)

(event EventTick
  (params (DeltaSeconds :type Float%20%28single-precision%29))
  (|CallParentFunction DeltaSeconds) ; This node is disabled and will not be called.\nDrag off pins to build functionality.
)

(event EventT_HideCurve__UpdateFunc
  (bind _self self)
  (bind _returnvalue (Transformation|GetActorLocation _self))
  (bind _returnvalue_2 (+ _returnvalue "0, 0, 0"))
  (|CallParentFunction)
  (Utilities|FlowControl|Sequence
    (:then_0
      (Utilities|FlowControl|DoOnce false
        (:Completed
          _)))
    (:then_1
      (Utilities|FlowControl|DoOnce false
        (:Completed
          (Niagara|SpawnSystematLocation "/BattlePassS18/QuestActors/Assets/FX/NS_QuestS18_Crystal_Success.NS_QuestS18_Crystal_Success" _returnvalue_2) ; SuccessFX
        ))))
  (bind _returnvalue_3 (Niagara|SpawnSystematLocation "/BattlePassS18/QuestActors/Assets/FX/NS_QuestS18_Crystal_Spawn.NS_QuestS18_Crystal_Spawn" _returnvalue_2)) ; Spawn/DespawnFX
  (Effects|Components|ParticleSystem|SetFloatParameter _returnvalue_3 "Scale" 3.0))

(event EventT_UnHideCurve__UpdateFunc
  (bind _self self)
  (bind _returnvalue (Transformation|GetActorLocation _self))
  (bind _returnvalue_2 (+ _returnvalue "0, 0, 0"))
  (|CallParentFunction)
  (Utilities|FlowControl|Sequence
    (:then_0
      (Utilities|FlowControl|DoOnce false
        (:Completed
          (Niagara|SpawnSystematLocation "/BattlePassS18/QuestActors/Assets/FX/NS_QuestS18_Crystal_Success.NS_QuestS18_Crystal_Success" _returnvalue_2) ; SuccessFX
        )))
    (:then_1
      (Utilities|FlowControl|DoOnce false
        (:Completed
          _))))
  (bind _returnvalue_3 (Niagara|SpawnSystematLocation "/BattlePassS18/QuestActors/Assets/FX/NS_QuestS18_Crystal_Spawn.NS_QuestS18_Crystal_Spawn" _returnvalue_2)) ; Spawn/DespawnFX
  (Effects|Components|ParticleSystem|SetFloatParameter _returnvalue_3 "Scale" 3.0))

(fn ConstructionScript
  (|CallParentFunction))
