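; Cosmetic buddy actor for a Season 18 Fortnitemares dig-up quest that rises out of the ground, then removes itself.
; BeginPlay registers its ambient sound component with the MusicEventSubsystem as a diegetic POI event.
; It then stores a start point below its spawn location, and the Emerge timeline lerps it upward to slightly above that spawn point with zeroed rotation.
; The actor destroys itself when the timeline finishes.
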(event EventBeginPlay
  (Utilities|FlowControl|Sequence
    (:then_0
      (bind _buddy_ambient_sound (Variables|Default|GetBuddy_Ambient_Sound)
        (:then
          (Audio|AddEventbyComponent (GameInstanceSubsystems|GetMusicEventSubsystem) _buddy_ambient_sound "(TagName=\"MusicEventSubsystem.Event.POI.Diegetic\")" "(GameplayTags=)" 0 ""))
        (:else)))
    (:then_1
      (Variables|Default|SetInitialPosition (+ (Transformation|GetActorLocation) "0, 0,-32.000000"))
      (bind (_direction _alpha) (Timeline|Emerge PlayFromStart)
        (:Update
          (Transformation|SetActorLocationAndRotation self (Math|Vector|Lerp(Vector) (Variables|Default|GetInitialPosition) (+ (Variables|Default|GetInitialPosition) "0, 0,48.000000") _alpha) "0, 0, 0" false false))
        (:Finished
          (Actor|DestroyActor))))))

(event Collision|EventActorBeginOverlap
  (params (OtherActor :type Actor%20Object%20Reference)))

(event EventTick
  (params (DeltaSeconds :type Float%20%28single-precision%29)))

(fn ConstructionScript)
