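; Cosmetic actor for a Season 18 Fortnitemares battle pass quest, apparently a hay child that rises out of a dig spot.
; On BeginPlay it stores a start position offset sideways and below its spawn location, then plays the Emerge timeline.
; The timeline lerps the actor straight up from that point with zero rotation, then destroys it when finished.
; Overlap, Tick and the construction script are empty.
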(event EventBeginPlay
  (Variables|Default|SetInitialPosition (+ (Transformation|GetActorLocation) "-48.000000,0.000000,-32.000000"))
  (bind (_direction _alpha) (Timeline|Emerge PlayFromStart)
    (:Update
      (Transformation|SetActorLocationAndRotation self (Math|Vector|Lerp(Vector) (Variables|Default|GetInitialPosition) (+ (Variables|Default|GetInitialPosition) "0, 0,48.000000") _alpha) "0, 0, 0" false false))
    (:Finished
      (Actor|DestroyActor))))

(event Collision|EventActorBeginOverlap
  (params (OtherActor :type Actor%20Object%20Reference)))

(event EventTick
  (params (DeltaSeconds :type Float%20%28single-precision%29)))

(fn ConstructionScript)
